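# Material function for the ByteGear character body that generates animated columns of colored pixels scrolling over the mesh.
# Texture UVs are swizzled or rotated either side of a U threshold, then feed two offset column layers masked with Frac, Floor and Abs.
# Per-column VectorNoise derives random pixel size and duration, varying the time-driven pattern.
# It outputs both tinted layers summed, plus the second alone; virtual-plane UV and flow-direction branches appear unused.
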MaterialFunction /BRCosmetics/Characters/Player/Male/Medium/Bodies/M_MED_ByteGear/FX/MF_ByteGearPixels
graph {nodes [
  n1 = TextureCoordinate()
  n2 = MaterialFunctionCall(XYZ=@n1) {materialFunction=&Swizzle}
  n3 = Constant() {r=-2}
  n4 = If(
   A=ComponentMask(None=@n1) {r=true}
   "A > B"=@n2.YX
   "A == B"=@n2.YX
   "A < B"=Rotator(Coordinate=@n1, Time=MaterialFunctionCall(Multiplier=@n3) {materialFunction=&Pi}.Result)
  ) {constB=0.384}
  n5 = ComponentMask(None=@n4) {g=true}
  n6 = ScalarParameter() {defaultValue=52, parameterName=Number of Columns}
  n7 = NamedRerouteUsage()
  n8 = Divide(A=@n6, B=@n7)
  n9 = Time()
  n10 = NamedRerouteUsage()
  n11 = Multiply(A=Frac(None=Divide(A=@n9, B=@n10) {constB=12}), B=@n8)
  n12 = ConstantBiasScale(None=Subtract(A=Multiply(A=@n5, B=@n8), B=Floor(None=@n11))) {bias=-0.5, scale=2}
  Add(A=@n12) {constB=0, desc=offset the pixels on Y-axis}
  n13 = LocalPosition()
  Add(A=@n13.XYZ, B=Constant3Vector() {constant=<rgba 0 120 0 0>})
  n14 = Add(A=@n4, B=Constant2Vector() {g=-0.2, r=-0.5})
  n15 = ComponentMask(None=@n14) {g=true}
  n16 = Constant() {r=51}
  n17 = NamedRerouteUsage()
  n18 = Divide(A=@n16, B=@n17)
  n19 = Time()
  n20 = NamedRerouteUsage()
  n21 = Multiply(A=Frac(None=Divide(A=@n19, B=@n20) {constB=12}), B=@n18)
  n22 = ConstantBiasScale(None=Subtract(A=Multiply(A=@n15, B=@n18), B=Floor(None=@n21))) {bias=-0.5, scale=2}
  Add(A=@n22) {constB=0, desc=offset the pixels on Y-axis}
  n23 = ComponentMask(None=Multiply(A=@n14, B=@n16)) {r=true}
  n24 = Divide(A=ConstantBiasScale(None=Frac(None=@n23)) {bias=-0.5, scale=2}, B=@n17)
  n25 = Saturate(None=Abs(None=AppendVector(A=@n24, B=@n22)))
  n26 = MaterialFunctionCall(Float2=OneMinus(None=Floor(None=@n25))) {materialFunction=&BreakOutFloat2Components}
  n27 = Multiply(A=Multiply(A=@n26.R, B=@n26.R), B=VectorParameter() {parameterName=Pixel Color, defaultValue=<rgba 0.005 0.3177 0.0308 1>}.RGB)
  FunctionOutput(None=@n27) {outputName=Result_1}
  n28 = ComponentMask(None=Multiply(A=@n4, B=@n6)) {r=true}
  n29 = Divide(A=ConstantBiasScale(None=Frac(None=@n28)) {bias=-0.5, scale=2}, B=@n7)
  n30 = Saturate(None=Abs(None=AppendVector(A=@n29, B=@n12)))
  n31 = MaterialFunctionCall(Float2=OneMinus(None=Floor(None=@n30))) {materialFunction=&BreakOutFloat2Components}
  n32 = Multiply(A=@n31.R, B=@n31.R)
  n33 = VectorParameter() {parameterName=Pixel Color, defaultValue=<rgba 0.005 0.3177 0.0308 1>}
  FunctionOutput(None=Add(A=Multiply(A=@n32, B=@n33.RGB), B=@n27))
  Multiply(A=@n22)
  Multiply(A=@n12)
  n34 = VectorNoise(World Position=@n28) {desc=Random RGB for different aspects}
  NamedRerouteDeclaration(None=LinearInterpolate(
    A=ScalarParameter() {defaultValue=0.35, parameterName=Random Pixel Size Min}
    B=ScalarParameter() {defaultValue=0.7, parameterName=Random Pixel Size Max}
    Alpha=ComponentMask(None=@n34) {desc=Random Pixel Size, r=true}
   )) {name=Random Pixel Size, nodeColor=<rgba 0.9922 0 1 1>}
  n35 = ScalarParameter() {defaultValue=48, parameterName=Random Duration Min}
  n36 = ScalarParameter() {defaultValue=62, parameterName=Random Duration Max}
  n37 = ComponentMask(None=@n34) {desc=Random Duration, g=true}
  n38 = ComponentMask(None=@n34) {b=true, desc=Random Direction}
  NamedRerouteDeclaration(None=Multiply(A=LinearInterpolate(A=@n35, B=@n36, Alpha=@n37) {constAlpha=1}, B=Max(A=@n38) {constB=0.5})) {name=Random Duration, nodeColor=<rgba 1 0.3516 0 1>}
  n39 = Constant3Vector() {constant=<rgba 0.1 0.1 0.1 0>}
  NamedRerouteDeclaration(None=ComponentMask(None=Multiply(A=@n13.XYZ, B=@n39)) {g=true, r=true}) {name=PlanarUV, nodeColor=<rgba 0.6875 0 1 1>}
  n40 = VectorNoise(World Position=@n23) {desc=Random RGB for different aspects}
  NamedRerouteDeclaration(None=LinearInterpolate(
    A=ScalarParameter() {defaultValue=0.35, parameterName=Random Pixel Size Min}
    B=ScalarParameter() {defaultValue=0.8, parameterName=Random Pixel Size Max}
    Alpha=ComponentMask(None=@n40) {desc=Random Pixel Size, r=true}
   )) {name=Random Pixel Size_1, nodeColor=<rgba 0.9922 0 1 1>}
  n41 = ScalarParameter() {defaultValue=48, parameterName=Random Duration Min}
  n42 = ScalarParameter() {defaultValue=62, parameterName=Random Duration Max}
  n43 = ComponentMask(None=@n40) {desc=Random Duration, g=true}
  n44 = ComponentMask(None=@n40) {b=true, desc=Random Direction}
  NamedRerouteDeclaration(None=Multiply(A=LinearInterpolate(A=@n41, B=@n42, Alpha=@n43) {constAlpha=1}, B=Max(A=@n44) {constB=0.5})) {name=Random Duration_1, nodeColor=<rgba 1 0.3516 0 1>}
  n45 = CameraVectorWS()
  n46 = Constant3Vector() {constant=<rgba 0 -1 0 0>}
  n47 = Normalize(VectorInput=CrossProduct(A=@n45, B=Transform(None=@n46) {transformSourceType=TRANSFORMSOURCE_View}))
  n48 = MaterialFunctionCall() {materialFunction=&ComponentPivotLocation}
  n49 = ScalarParameter() {defaultValue=200, parameterName=UV Scale}
  NamedRerouteDeclaration(None=Multiply(
    A=MaterialFunctionCall(Plane Normal Axis=@n45, Plane X Axis=@n47, Plane Center=@n48.Result, UV Scale=@n49) {materialFunction=&VirtualPlaneCoordinates}.UVs
    B=Constant2Vector() {g=1, r=1}
   )) {name=VirtualPlaneUVs, nodeColor=<rgba 0 1 0.2031 1>}
  NamedRerouteUsage()
  NamedRerouteUsage()
  ScalarParameter() {defaultValue=0.009, parameterName=PlanarUVScale}
  Sign(None=Subtract(A=@n38) {constB=0.5}) {desc="Negative flows up, Positive flows down"}
  Sign(None=Subtract(A=@n44) {constB=0.5}) {desc="Negative flows up, Positive flows down"}
 ]}
