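# Niagara system for the ByteGear character's roll idle effect, showing short-lived, multicolored glitch sprites around the body.
# Two near-identical sprite emitters, GlitchHard and GlitchSoft, split particles into seven partitions that select bone, offset and scale across five skeletal mesh bones or sockets.
# Particle update sets random bone-relative positions, stepped size curves and dynamic material parameters, with a camera offset.
# Spawn colors are scaled by time-of-day brightness.
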NiagaraSystem /BRCosmetics/Characters/Player/Male/Medium/Bodies/M_MED_ByteGear/FX/NS_Character_ByteGear_Roll_Idle
dependencies {
 usedDataInterfaces [
  &NiagaraDataInterfaceCurve
  &NiagaraDataInterfaceArrayColor
  &NiagaraDataInterfaceArrayFloat
  &NiagaraDataInterfaceArrayInt32
  &NiagaraDataInterfaceArrayFloat3
  &NiagaraDataInterfaceSkeletalMesh
 ]
 usedDynamicInputs [&FNTimeOfDayScalar]
 usedModules [
  &self:SystemSpawnScript.SystemScriptSource.SystemScriptGraph.NiagaraNodeAssignment_1.SetVariables_#1
  &SystemState
  &self:SystemSpawnScript.SystemScriptSource.SystemScriptGraph.NiagaraNodeAssignment_2.SetVariables_#2
 ]
}
emitters {
 GlitchHard {
  emitterUpdateScript [
   &EmitterState(Scalability Mode=::NewEnumerator1)
   &SpawnRate(SpawnRate=15, Use Spawn Probability=false)
  ]
  particleSpawnScript [
   &PartitionParticles(Number Of Partitions=7)
   &InitializeParticle(
    Color=&Multiply_LinearColorByFloat(
     Linear Color=&SelectLinearColorFromArray(
      Array Sampling Mode=::NewEnumerator0
      Direct Array Index=?
      Color Selection Array=DataInterface() {propertyValues {ColorData [<rgba 0.65 0.0003 0.0693 1>, <rgba 0.105 0.3 0.171 1>, <rgba 0.65 0.0103 0.3577 1>, <rgba 0.1049 0.0663 0.35 1>]}}
      Random Seed=0
     )
     Float=$System.TimeOfDayBrightnessMultiplier
    )
    Lifetime=&RandomRangeFloat(Minimum=0.3, Maximum=0.6)
    Sprite Size=&Lerp_Vector2(
     A=&RandomRangeVector2D(Minimum=<3 3>, Maximum=<11 11>)
     B=&RandomRangeVector2D(
      Minimum=&Lerp_Vector2(
       A=<5 1.5>
       B=<1.5 5>
       Alpha=&UniformAOrBFloat(A=0, Distribution Weight=0.15, Evaluation Type=::NewEnumerator0, Recalculate Random Each Loop=true, Override Seed=false)
      )
      Maximum=&Lerp_Vector2(
       A=<18 4>
       B=<4 18>
       Alpha=&UniformAOrBFloat(A=0, Distribution Weight=0.15, Evaluation Type=::NewEnumerator0, Recalculate Random Each Loop=true, Override Seed=false)
      )
     )
     Alpha=&UniformAOrBFloat(A=0, Distribution Weight=0.7, Evaluation Type=::NewEnumerator0, Recalculate Random Each Loop=true, Override Seed=false)
    )
    Sprite Size Mode=::NewEnumerator4
    UsePositionOffset=false
    Write Material Random=true
   ) {module=InitializeParticle001}
   &DynamicMaterialParameters(
    Index 0 Param 2=&RandomRangeFloat(Minimum=0, Maximum=1)
    Index 0 Param 4=&RandomRangeFloat(Minimum=0.1, Maximum=0.2)
    Index 1 Param 1=&RandomRangeFloat(Minimum=-100, Maximum=100)
    Module.Param0WriteEnabled=false
    Module.Param0WriteEnabled1=true
    Module.Param1WriteEnabled=true
    Module.Param1WriteEnabled1=false
    Module.Param2WriteEnabled=false
    Module.Param2WriteEnabled1=false
    Module.Param3WriteEnabled=true
    Module.Param3WriteEnabled1=false
    Write Parameter Index 1=true
   ) {module=DynamicMaterialParameters001}
  ]
  particleUpdateScript [
   &SolveForcesAndVelocity(Clamp Velocity=false, Limit Acceleration=false)
   &ParticleState()
   &ScaleSpriteSize(
    Scale Sprite Size Mode=::NewEnumerator0
    Uniform Scale Factor=&SelectFloatFromArray(
     Direct Array Index=$Particles.PartitionParticles.Partition
     Float Selection Array=DataInterface() {propertyValues {FloatData [1,1,1,0.5,0.5,0.75,0.75]}}
    )
   ) {module=ScaleSpriteSize002}
   &ScaleSpriteSize(Uniform Curve Sprite Scale=DataInterface() {propertyValues {Curve {keys [<at 0 0 RCIM_Constant>,<at 0.2 0.35 RCIM_Constant>,<at 0.5 0.675 RCIM_Constant>,<at 0.8 1 RCIM_Constant>,<at 1 0 RCIM_Constant>]}}})
   &ScaleSpriteSize(Uniform Curve Sprite Scale=DataInterface() {propertyValues {Curve {keys [<at 0 1 RCIM_Cubic>,<at 0.5 1.3 RCIM_Cubic>,<at 1 1 RCIM_Cubic>]}}}) {module=ScaleSpriteSize001}
   &SkeletalMeshLocation(
    Bone / Socket Index=&SelectIntFromArray(Direct Array Index=$Particles.PartitionParticles.Partition, Int Selection Array=DataInterface() {propertyValues {IntData [0,0,0,1,2,3,4]}})
    Bone Sampling Mode=::NewEnumerator4
    MaskSpawnGroup=false
    Orientation Sampling=::NewEnumerator1
    Sampled Position Offset=&AddVectorToPosition(
     Position=&Multiply_Vector(
      A=&MergeHelper_VectorSwitch(
       0=&RandomRangeVector(Minimum=<-1.1 -0.4 -0.5>, Maximum=<1.1 0.4 0.5>)
       1=&RandomRangeVector(Minimum=<-1.1 -0.4 -0.5>, Maximum=<1.1 0.4 0.5>)
       2=&RandomRangeVector(Minimum=<-1.1 -0.4 -0.5>, Maximum=<1.1 0.4 0.5>)
       3=&RandomRangeVector(Minimum=<-1 0.4 -0.4>, Maximum=<1 0.4 0.4>)
       4=&RandomRangeVector(Minimum=<-1 -0.4 -0.4>, Maximum=<1 0.4 0.4>)
       5=&RandomRangeVector(Minimum=<-1 -0.6 -0.6>, Maximum=<1 0.6 0.6>)
       6=&RandomRangeVector(Minimum=<-1 -0.6 -0.6>, Maximum=<1 0.6 0.6>)
       Index=$Particles.PartitionParticles.Partition
       Num=::NewEnumerator5
      )
      B=&VectorFromFloat(Value=&SelectFloatFromArray(
        Direct Array Index=$Particles.PartitionParticles.Partition
        Float Selection Array=DataInterface() {propertyValues {FloatData [27,27,27,10,10,10,10]}}
       ))
     )
     Vector=&SelectVectorFromArray(
      Direct Array Index=$Particles.PartitionParticles.Partition
      Vector Selection Array=DataInterface() {propertyValues {FloatData [<17 0 0>, <17 0 0>, <17 0 0>, <-32 -2 0>, <32 2 0>, <5 0 0>, <-5 0 0>]}}
     )
    )
    Skeletal Mesh=$System.Skeletal Mesh
   )
   &DynamicMaterialParameters(
    Index 0 Param 2=&RandomRangeFloat(Minimum=1, Maximum=1)
    Index 0 Param 3=&UniformAOrBFloat(
     A=0.7
     B=&FloatFromCurve(FloatCurve=DataInterface() {propertyValues {Curve {keys [
          {arriveTangent=-0.7079, interpMode=RCIM_Cubic, time=0, value=1}
          {arriveTangent=-2.3718, interpMode=RCIM_Cubic, leaveTangent=-2.3718, tangentMode=RCTM_User, time=1, value=0}
         ]}}})
     Distribution Weight=1
     Evaluation Type=::NewEnumerator0
     Recalculate Random Each Loop=true
     Override Seed=false
    )
    Index 1 Param 4=&FloatFromCurve(FloatCurve=DataInterface() {propertyValues {Curve {keys [
         {arriveTangent=4.1785, interpMode=RCIM_Cubic, leaveTangent=4.1785, tangentMode=RCTM_User, time=0, value=0}
         {arriveTangent=0.9995, interpMode=RCIM_Cubic, leaveTangent=0.9995, tangentMode=RCTM_User, time=0.2, value=0.525}
         {arriveTangent=-0.0405, interpMode=RCIM_Cubic, leaveTangent=-0.0405, tangentMode=RCTM_User, time=1, value=1}
        ]}}})
    Module.Param0WriteEnabled=false
    Module.Param0WriteEnabled1=false
    Module.Param1WriteEnabled=false
    Module.Param1WriteEnabled1=false
    Module.Param2WriteEnabled=true
    Module.Param2WriteEnabled1=false
    Module.Param3WriteEnabled=false
    Module.Param3WriteEnabled1=true
    Write Parameter Index 1=true
   ) {module=DynamicMaterialParameters002}
   &CameraOffset(Camera Offset Amount=8)
  ]
  rendererClasses [&NiagaraSpriteRendererProperties]
  renderers [{rendererClass=&NiagaraSpriteRendererProperties}]
 }
 GlitchSoft {
  emitterUpdateScript [
   &EmitterState(Scalability Mode=::NewEnumerator1)
   &SpawnRate(SpawnRate=15, Use Spawn Probability=false)
  ]
  particleSpawnScript [
   &PartitionParticles(Number Of Partitions=7)
   &InitializeParticle(
    Color=&Multiply_LinearColorByFloat(
     Linear Color=&SelectLinearColorFromArray(
      Array Sampling Mode=::NewEnumerator0
      Direct Array Index=?
      Color Selection Array=DataInterface() {propertyValues {ColorData [<rgba 0.7529 0.0003 0.0802 1>, <rgba 0.1248 0.3515 0.2016 1>, <rgba 0.7305 0.0116 0.402 1>, <rgba 0.1301 0.0823 0.4342 1>]}}
      Random Seed=0
     )
     Float=$System.TimeOfDayBrightnessMultiplier
    )
    Lifetime=&RandomRangeFloat(Minimum=0.3, Maximum=0.6)
    Sprite Size=&Lerp_Vector2(
     A=&RandomRangeVector2D(Minimum=<3 3>, Maximum=<11 11>)
     B=&RandomRangeVector2D(
      Minimum=&Lerp_Vector2(
       A=<5 1.5>
       B=<1.5 5>
       Alpha=&UniformAOrBFloat(A=0, Distribution Weight=0.15, Evaluation Type=::NewEnumerator0, Recalculate Random Each Loop=true, Override Seed=false)
      )
      Maximum=&Lerp_Vector2(
       A=<18 4>
       B=<4 18>
       Alpha=&UniformAOrBFloat(A=0, Distribution Weight=0.15, Evaluation Type=::NewEnumerator0, Recalculate Random Each Loop=true, Override Seed=false)
      )
     )
     Alpha=&UniformAOrBFloat(A=0, Distribution Weight=0.7, Evaluation Type=::NewEnumerator0, Recalculate Random Each Loop=true, Override Seed=false)
    )
    Sprite Size Mode=::NewEnumerator4
    UsePositionOffset=false
    Write Material Random=true
   ) {module=InitializeParticle001}
   &DynamicMaterialParameters(
    Index 0 Param 2=&RandomRangeFloat(Minimum=0, Maximum=1)
    Index 0 Param 4=&RandomRangeFloat(Minimum=0.1, Maximum=0.2)
    Index 1 Param 1=&RandomRangeFloat(Minimum=-100, Maximum=100)
    Module.Param0WriteEnabled=false
    Module.Param0WriteEnabled1=true
    Module.Param1WriteEnabled=true
    Module.Param1WriteEnabled1=false
    Module.Param2WriteEnabled=false
    Module.Param2WriteEnabled1=false
    Module.Param3WriteEnabled=true
    Module.Param3WriteEnabled1=false
    Write Parameter Index 1=true
   ) {module=DynamicMaterialParameters001}
  ]
  particleUpdateScript [
   &SolveForcesAndVelocity(Clamp Velocity=false, Limit Acceleration=false)
   &ParticleState()
   &ScaleSpriteSize(
    Scale Sprite Size Mode=::NewEnumerator0
    Uniform Scale Factor=&SelectFloatFromArray(
     Direct Array Index=$Particles.PartitionParticles.Partition
     Float Selection Array=DataInterface() {propertyValues {FloatData [1,1,1,0.5,0.5,0.75,0.75]}}
    )
   ) {module=ScaleSpriteSize002}
   &ScaleSpriteSize(
    Uniform Curve Scale=0.8
    Uniform Curve Sprite Scale=DataInterface() {propertyValues {Curve {keys [<at 0 0 RCIM_Constant>,<at 0.2 0.35 RCIM_Constant>,<at 0.5 0.675 RCIM_Constant>,<at 0.8 1 RCIM_Constant>,<at 1 0 RCIM_Constant>]}}}
   )
   &ScaleSpriteSize(Uniform Curve Sprite Scale=DataInterface() {propertyValues {Curve {keys [<at 0 1 RCIM_Cubic>,<at 0.5 1.3 RCIM_Cubic>,<at 1 1 RCIM_Cubic>]}}}) {module=ScaleSpriteSize001}
   &SkeletalMeshLocation(
    Bone / Socket Index=&SelectIntFromArray(Direct Array Index=$Particles.PartitionParticles.Partition, Int Selection Array=DataInterface() {propertyValues {IntData [0,0,0,1,2,3,4]}})
    Bone Sampling Mode=::NewEnumerator4
    MaskSpawnGroup=false
    Orientation Sampling=::NewEnumerator1
    Sampled Position Offset=&AddVectorToPosition(
     Position=&Multiply_Vector(
      A=&MergeHelper_VectorSwitch(
       0=&RandomRangeVector(Minimum=<-1.2 -0.4 -0.5>, Maximum=<1.2 0.4 0.5>)
       1=&RandomRangeVector(Minimum=<-1.2 -0.4 -0.5>, Maximum=<1.2 0.4 0.5>)
       2=&RandomRangeVector(Minimum=<-1.2 -0.4 -0.5>, Maximum=<1.2 0.4 0.5>)
       3=&RandomRangeVector(Minimum=<-1 0.4 -0.4>, Maximum=<1 0.4 0.4>)
       4=&RandomRangeVector(Minimum=<-1 -0.4 -0.4>, Maximum=<1 0.4 0.4>)
       5=&RandomRangeVector(Minimum=<-1 -0.6 -0.6>, Maximum=<1 0.6 0.6>)
       6=&RandomRangeVector(Minimum=<-1 -0.6 -0.6>, Maximum=<1 0.6 0.6>)
       Index=$Particles.PartitionParticles.Partition
       Num=::NewEnumerator5
      )
      B=&VectorFromFloat(Value=&SelectFloatFromArray(
        Direct Array Index=$Particles.PartitionParticles.Partition
        Float Selection Array=DataInterface() {propertyValues {FloatData [27,27,27,10,10,10,10]}}
       ))
     )
     Vector=&SelectVectorFromArray(
      Direct Array Index=$Particles.PartitionParticles.Partition
      Vector Selection Array=DataInterface() {propertyValues {FloatData [<19 0 0>, <19 0 0>, <19 0 0>, <-32 -2 0>, <32 2 0>, <5 0 0>, <-5 0 0>]}}
     )
    )
    Skeletal Mesh=$System.Skeletal Mesh
   )
   &DynamicMaterialParameters(
    Index 0 Param 2=&RandomRangeFloat(Minimum=1, Maximum=1)
    Index 0 Param 3=&UniformAOrBFloat(
     A=0.7
     B=&FloatFromCurve(FloatCurve=DataInterface() {propertyValues {Curve {keys [
          {arriveTangent=-0.7079, interpMode=RCIM_Cubic, time=0, value=1}
          {arriveTangent=-2.3718, interpMode=RCIM_Cubic, leaveTangent=-2.3718, tangentMode=RCTM_User, time=1, value=0}
         ]}}})
     Distribution Weight=1
     Evaluation Type=::NewEnumerator0
     Recalculate Random Each Loop=true
     Override Seed=false
    )
    Index 1 Param 4=&FloatFromCurve(FloatCurve=DataInterface() {propertyValues {Curve {keys [
         {arriveTangent=4.1785, interpMode=RCIM_Cubic, leaveTangent=4.1785, tangentMode=RCTM_User, time=0, value=0}
         {arriveTangent=0.9995, interpMode=RCIM_Cubic, leaveTangent=0.9995, tangentMode=RCTM_User, time=0.2, value=0.525}
         {arriveTangent=-0.0405, interpMode=RCIM_Cubic, leaveTangent=-0.0405, tangentMode=RCTM_User, time=1, value=1}
        ]}}})
    Module.Param0WriteEnabled=false
    Module.Param0WriteEnabled1=false
    Module.Param1WriteEnabled=false
    Module.Param1WriteEnabled1=false
    Module.Param2WriteEnabled=true
    Module.Param2WriteEnabled1=false
    Module.Param3WriteEnabled=false
    Module.Param3WriteEnabled1=true
    Write Parameter Index 1=true
   ) {module=DynamicMaterialParameters002}
   &CameraOffset(Camera Offset Amount=8)
  ]
  rendererClasses [&NiagaraSpriteRendererProperties]
  renderers [{rendererClass=&NiagaraSpriteRendererProperties}]
 }
}
summary {
 systemName=NS_Character_ByteGear_Roll_Idle
 emitters [
  {emitterName=GlitchHard, rendererClasses [&NiagaraSpriteRendererProperties]}
  {emitterName=GlitchSoft, rendererClasses [&NiagaraSpriteRendererProperties]}
 ]
 userVariables [
  User.bisFootHidden {type=NiagaraBool}
  User.bIsFrontEnd {type=NiagaraBool}
  User.bIsFrontEndPreview {type=NiagaraBool}
 ]
}
