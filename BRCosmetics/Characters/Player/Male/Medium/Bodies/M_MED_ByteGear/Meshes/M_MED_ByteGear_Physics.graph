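# Physics asset for the ByteGear medium male player body, apparently for secondary motion of a console accessory at the right thigh.
# Capsules on pelvis, spine_01 and right thigh form a partial body, with both spine and thigh joined to the pelvis by free-rotation constraints.
# A tiny deform_console capsule links freely to the thigh and anchors dyn_console, a two-capsule body whose joint has tighter swing and twist limits.
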PhysicsAsset /BRCosmetics/Characters/Player/Male/Medium/Bodies/M_MED_ByteGear/Meshes/M_MED_ByteGear_Physics
bodies {
 deform_console [{shapeType=Capsule, length=1, radius=1, shapeName=deform_console_capsule}]
 dyn_console [
  {shapeType=Capsule, length=9.3007, radius=7.0038, shapeName=dyn_console_capsule, center=<9.7796 -6.821 0>}
  {shapeType=Capsule, length=9.3007, radius=7.0038, shapeName=dyn_console_capsule, center=<-1.863 -6.3342 0>}
 ]
 pelvis [{shapeType=Capsule, length=1.6719, radius=13, shapeName=pelvis_capsule, center=<-4.5493 -4.509 2.3545>}]
 spine_01 [{shapeType=Capsule, length=12.3562, radius=11.8668, shapeName=spine_01_capsule, center=<4.1408 0 0>, rotation=<rot 0 -90 0>}]
 thigh_r [{
   shapeType=Capsule
   length=32.6845
   radius=11.2046
   shapeName=thigh_r_capsule
   center=<17.1305 -0.4642 -2.8879>
   rotation=<rot 87.1875 179.9993 179.9996>
  }]
}
constraints [
 {
  bone1Name=thigh_r
  bone2Name=pelvis
  swing1LimitDegrees=45
  swing1Motion=Free
  swing2LimitDegrees=45
  swing2Motion=Free
  twistLimitDegrees=45
  twistMotion=Free
 }
 {
  bone1Name=deform_console
  bone2Name=thigh_r
  swing1LimitDegrees=45
  swing1Motion=Free
  swing2LimitDegrees=45
  swing2Motion=Free
  twistLimitDegrees=45
  twistMotion=Free
 }
 {bone1Name=dyn_console, bone2Name=deform_console, swing1LimitDegrees=30, swing2LimitDegrees=15, twistLimitDegrees=30}
 {
  bone1Name=spine_01
  bone2Name=pelvis
  swing1LimitDegrees=45
  swing1Motion=Free
  swing2LimitDegrees=45
  swing2Motion=Free
  twistLimitDegrees=45
  twistMotion=Free
 }
]
