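# Post-process-style animation Blueprint for the male medium ByteGear body, which copies the attached parent mesh's pose and adds secondary motion.
# A Control Rig, given skydiving and performance-mode flags, feeds pelvis-space RigidBody physics with overridden gravity.
# While baseball sliding, blended additive bone offsets reposition the console and wire, then spring-driven AnimDynamics moves a wire bone.
# The EventGraph updates a Gravity Override variable through a deprecated function, though no AnimGraph node reads it.
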AnimBlueprint /BRCosmetics/Characters/Player/Male/Medium/Bodies/M_MED_ByteGear/Meshes/M_MED_ByteGear_AnimBP
target_skeleton=&/M_MED_ByteGear_Skeleton
graphs [
 AnimGraph {nodes [
   n1 = Control Rig!(
    Source=Copy Pose From Mesh() {bUseAttachedParent=true}
    Alpha
    Skydiving:bool=Get bIsSkyDiving()
    PerformanceMode:bool=NOT Boolean(A=Get bRigidBodyEnabled())
   ) -> (Pose) {
    alphaBoolBlend {blendOption=Linear}
    controlRigAssetReference {blueprintRigClass=&/M_MED_ByteGear_CtrlRig.M_MED_ByteGear_CtrlRig_C}
   }
   n2 = RigidBody!(ComponentPose=Local To Component(LocalPose=@n1), Alpha, OverrideWorldGravity:Vector="0,0,0") -> (Pose) {
    bClampLinearTranslationLimitToRefPose=true
    bOverrideWorldGravity=true
    overridePhysicsAsset=&/M_MED_ByteGear_Physics
    simulationSpace=BaseBoneSpace
    alphaBoolBlend {blendInTime=1.7, blendOption=Linear, blendOutTime=1.7}
    baseBoneRef {boneName=pelvis}
    componentAppliedLinearAccClamp=<3000 3000 3000>
    overrideWorldGravity {z=-980}
    simSpaceSettings {worldAlpha=0.4}
   }
   n3 = Get bIsBaseballSliding()
   n4 = "Transform (Modify) Bone - Bone: deform_console"!(ComponentPose=@n2, bAlphaBoolEnabled:bool=@n3, Translation="-0.759,11.937,-10.655", Rotation="-41.070,-16.265,-54.849", Scale="1,1,1") -> (Pose) {
    alphaInputType=Bool
    rotationMode=BMM_Additive
    rotationSpace=BCS_BoneSpace
    translationMode=BMM_Additive
    translationSpace=BCS_BoneSpace
    alphaBoolBlend {blendInTime=1, blendOption=Linear, blendOutTime=0.3}
    boneToModify {boneName=deform_console}
    rotation=<rot -41.07 -16.265 -54.8489>
    translation=<-0.759 11.937 -10.655>
   }
   n5 = "Transform (Modify) Bone - Bone: deform_console_wire"!(ComponentPose=@n4, bAlphaBoolEnabled:bool=@n3, Translation="0,0,0", Rotation="51.202,24.605,50.285", Scale="1,1,1") -> (Pose) {
    alphaInputType=Bool
    rotationMode=BMM_Additive
    rotationSpace=BCS_BoneSpace
    translationMode=BMM_Additive
    translationSpace=BCS_BoneSpace
    alphaBoolBlend {blendInTime=1, blendOption=Linear, blendOutTime=0.3}
    boneToModify {boneName=deform_console_wire}
    rotation=<rot 51.202 24.605 50.285>
   }
   n6 = "Transform (Modify) Bone - Bone: dyn_console_1"!(ComponentPose=@n5, bAlphaBoolEnabled:bool=@n3, Translation="0,0,0", Rotation="-9.657,15.221,-2.611", Scale="1,1,1") -> (Pose) {
    alphaInputType=Bool
    rotationMode=BMM_Additive
    rotationSpace=BCS_BoneSpace
    translationMode=BMM_Additive
    translationSpace=BCS_BoneSpace
    alphaBoolBlend {blendInTime=1, blendOption=Linear, blendOutTime=0.3}
    boneToModify {boneName=dyn_console_1}
    rotation=<rot -9.657 15.221 -2.611>
   }
   n7 = Anim Dynamics - Bone: deform_console_wire_1!(ComponentPose=@n6, Alpha=1, GravityOverride:Vector="1,1,1") -> (Pose) {
    angularDampingOverride=0.998
    angularSpringConstant=1500
    bAngularSpring=true
    bGravityOverrideInSimSpace=true
    bLinearSpring=true
    bOverrideAngularDamping=true
    bOverrideLinearDamping=true
    bUseGravityOverride=true
    linearDampingOverride=1
    linearSpringConstant=500
    simulationSpace=World
    alphaBoolBlend {blendOption=Linear}
    boundBone {boneName=deform_console_wire_1}
    componentAppliedLinearAccClamp=<200 200 200>
    gravityOverride=<1 1 1>
    physicsBodyDefinitions [{
      sphereCollisionRadius=0
      boundBone {boneName=deform_console_wire_1}
      boxExtents=<12 8 12>
      constraintSetup {angularLimitsMax=<0 0 0>, angularLimitsMin=<0 -10 -5>, angularTarget=<1 0 0>, linearAxesMax=<0 0 3>, linearAxesMin=<0 0 -1>}
      localJointOffset=<0.1 0 0>
     }]
    relativeSpaceBone {boneName=spine_05}
   }
   Output Pose(Result=Component To Local(ComponentPose=@n7)) {layerGroup=None}
  ]}
 EventGraph {nodes [
   Event BlueprintUpdateAnimation()
   > n1 = Gravity Override State Function DEPRECATED(CharacterPartAnimInstance=Self-Reference(), GameTimeInSeconds=GetGameTimeInSeconds())
   > Set Gravity Override(Gravity Override=@n1)
  ]}
]
variables [Gravity Override {type=Vector}]
