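# Control Rig for the face accessory of the male medium ByteGear player body, procedurally correcting its wire bones as the head moves.
# It aims deform_wire at spine_05, then clamps dyn_wire_2 spatially in a cylinder around clavicle_l to keep it clear of that bone.
# Two spherical pose readers with opposed orientations measure head relative to neck_01 and weight corrective offsets on dyn_wire_1 and dyn_wire_2.
# Debug drawing remains enabled on the clamp and second reader.
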ControlRigBlueprint /BRCosmetics/Characters/Player/Male/Medium/Bodies/M_MED_ByteGear/Meshes/Parts/M_MED_ByteGear_FaceAcc_CtrlRig
graphs [
 RigVMFunctionLibrary {links [], nodes []}
 RigVMModel {nodes [
   RigUnit_BeginExecution = &RigUnit_BeginExecution()
   &RigUnit_AimConstraintLocalSpaceOffset(
    Child=(Type=Bone,Name="deform_wire")
    Filter=(bX=True,bY=True,bZ=True)
    Parents=((Item=(Type=Bone,Name="spine_05"),Weight=1))
    UpAxis=(X=0,Y=-1,Z=0)
    Weight=1
    WorldUp=(Target=(X=0,Y=0,Z=1),Kind=Location,Space=(Type=Bone,Name="deform_wire"))
    ExecutePin=@RigUnit_BeginExecution.ExecutePin
   )
   > Set Transform_2 = /Script/ControlRig.RigUnit_SetTranslation(
    Item=(Type=Bone,Name="dyn_wire_2")
    Space=GlobalSpace
    Weight=1
    bInitial=false
    bPropagateToChildren=true
    Value=/Script/RigVM.RigVMFunction_MathTransformClampSpatially(
     Axis=X
     DebugColor=(R=1,G=0,B=0,A=1)
     DebugThickness=1
     Maximum=150
     Minimum=10
     Type=Cylinder
     bDrawDebug=true
     Value=&RigUnit_GetTransform(Item=(Type=Bone,Name="dyn_wire_2"), bInitial=false).Transform
     Space=&RigUnit_GetTransform(Item=(Type=Bone,Name="clavicle_l"), bInitial=false).Transform
    ) {
     is=struct RigVMUnitNode
     template="ClampSpatially::Execute(in Value,in Axis,in Type,in Minimum,in Maximum,in Space,in bDrawDebug,in DebugColor,in DebugThickness,out Result)"
    }.Result.Translation
   ) {is=struct RigVMUnitNode, template="Set Transform::Execute(in Item,in Space,in bInitial,in Value,in Weight,in bPropagateToChildren)"}
   > n1 = &RigUnit_SphericalPoseReader(
    ActiveRegionScaleFactors=(PositiveWidth=1,NegativeWidth=1,PositiveHeight=1,NegativeHeight=1)
    ActiveRegionSize=0.2
    Debug=(bDrawDebug=false,bDraw2D=False,bDrawLocalAxes=False,DebugScale=25,DebugSegments=20,DebugThickness=0.25)
    DriverAxis=(X=0,Y=1,Z=0)
    DriverItem=(Type=Bone,Name="head")
    FalloffRegionScaleFactors=(PositiveWidth=1,NegativeWidth=1,PositiveHeight=1,NegativeHeight=1)
    FalloffSize=0.5
    FlipHeightScaling=False
    FlipWidthScaling=False
    OptionalParentItem=(Type=Bone,Name="neck_01")
    RotationOffset=(X=180,Y=0,Z=0)
   )
   > &RigUnit_ModifyTransforms(
    ItemToModify=((Item=(Type=Bone,Name="dyn_wire_2"),Transform=(Rotation=(X=0,Y=0,Z=0,W=1),Translation=(X=5,Y=0,Z=-5),Scale3D=(X=1,Y=1,Z=1))))
    WeightMinimum=0
    Weight=@n1.OutputParam
   )
   > n2 = &RigUnit_SphericalPoseReader(
    ActiveRegionScaleFactors=(PositiveWidth=1,NegativeWidth=1,PositiveHeight=1,NegativeHeight=1)
    ActiveRegionSize=0.2
    Debug=(bDrawDebug=true,bDraw2D=False,bDrawLocalAxes=False,DebugScale=25,DebugSegments=20,DebugThickness=0.25)
    DriverAxis=(X=0,Y=1,Z=0)
    DriverItem=(Type=Bone,Name="head")
    FalloffRegionScaleFactors=(PositiveWidth=1,NegativeWidth=1,PositiveHeight=1,NegativeHeight=1)
    FalloffSize=0.5
    FlipHeightScaling=False
    FlipWidthScaling=False
    OptionalParentItem=(Type=Bone,Name="neck_01")
    RotationOffset=(X=0,Y=0,Z=0)
   )
   > &RigUnit_ModifyTransforms(
    ItemToModify="((Item=(Type=Bone,Name=\"dyn_wire_1\"),Transform=(Rotation=(X=-0.5806,Y=0.1389,Z=0.2667,W=-0.7566),Translation=(X=-5.3738,Y=-1.3261,Z=5.6457),Scale3D=(X=1,Y=1,Z=1))),(Item=(Type=Bone,Name=\"dyn_wire_2\"),Transform=(Rotation=(X=0,Y=0.342,Z=0,W=-0.9397),Translat ...(truncated)"
    WeightMinimum=0
    Weight=@n2.OutputParam
   )
  ]}
]
hierarchy [
 root {loc=<0 0 0>, rot=<0 0 0 1>}
 pelvis {parent=root, loc=<0 -0.7071 93.1191>, rot=<0.0127 0.707 -0.0127 -0.707>}
 spine_01 {parent=pelvis, loc=<6.7494 -0.6388 0>, rot=<0 0 0.034 -0.9994>}
 spine_02 {parent=spine_01, loc=<7.2518 -0.035 0>, rot=<0 0 0.0238 -0.9997>}
 spine_03 {parent=spine_02, loc=<6.0104 0 0>, rot=<0 0 0.0149 -0.9999>}
 spine_04 {parent=spine_03, loc=<8.1024 0 0>, rot=<0 0 -0.0132 -0.9999>}
 spine_05 {parent=spine_04, loc=<10.0146 0 0>, rot=<0 0 -0.0015 -1>}
 neck_01 {parent=spine_05, loc=<11.3138 0.1529 0>, rot=<0 0 -0.2003 -0.9797>}
 neck_02 {parent=neck_01, loc=<4.5627 0 0>, rot=<0 0 0.0504 -0.9987>}
 head {parent=neck_02, loc=<5.3914 0 0>, rot=<0 0 0.115 -0.9934>}
 deform_wire {parent=head, loc=<4.8708 -5.913 4.1092>, rot=<0.2046 -0.9631 -0.0352 -0.1713>}
 dyn_wire {parent=deform_wire, loc=<0 0 0>, rot=<0 0 0 1>}
 dyn_wire_1 {parent=dyn_wire, loc=<16.3074 0 0>, rot=<-0.0723 0.5847 -0.0682 -0.8051>}
 dyn_wire_up_2 {parent=head, loc=<8.6517 2.5788 2.3412>, rot=<0.612 -0.1171 0.5256 -0.5792>}
 dyn_wire_up_1 {parent=head, loc=<9.2081 2.1784 1.4508>, rot=<0.603 -0.1571 0.4864 -0.6125>}
 clavicle_l {parent=spine_05, loc=<4.3575 -0.1466 -2.3223>, rot=<0.0465 -0.7289 -0.009 -0.6829>}
 clavicle_r {parent=spine_05, loc=<4.3575 -0.1466 2.3223>, rot=<0.7289 0.0465 -0.6829 0.009>}
 dyn_wire_2 {parent=spine_05, loc=<13.0005 -3.1094 -7.9074>, rot=<0.6961 -0.7088 0.0495 0.1024>}
 deform_console_wire {parent=spine_03, loc=<-4.0222 8.2571 10.749>, rot=<0.2942 -0.3358 0.8755 -0.1846>}
 thigh_r {parent=pelvis, loc=<-2.5693 -0.0109 7.7501>, rot=<0.0572 -0.0012 0.9982 0.0155>}
 thigh_twist_01_r {parent=thigh_r, loc=<21.6303 0 0>, rot=<0.0013 0 0 -1>}
 deform_console {parent=thigh_r, loc=<3.4652 -1.4556 8.0329>, rot=<0.5942 -0.0104 0.0307 -0.8036>}
 dyn_console {parent=deform_console, loc=<0 0 0>, rot=<0 0 0 1>}
 dyn_console_1 {parent=dyn_console, loc=<-3.538 -1.9292 2.7115>, rot=<0 0.1161 0 -0.9932>}
]
