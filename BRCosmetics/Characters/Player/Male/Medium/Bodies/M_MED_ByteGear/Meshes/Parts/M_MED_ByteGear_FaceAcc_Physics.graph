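# Physics asset for the face accessory of the male medium ByteGear character body, giving its dangling wire physical simulation.
# Capsules cover head, neck, upper spine and left clavicle, linked by free-rotating constraints, alongside capsules for the wire bones.
# The wire chain runs from deform_wire on the head through three dyn_wire bones with angular limits.
# Its last bone is also constrained to spine_05, tethering the wire at both ends.
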PhysicsAsset /BRCosmetics/Characters/Player/Male/Medium/Bodies/M_MED_ByteGear/Meshes/Parts/M_MED_ByteGear_FaceAcc_Physics
bodies {
 clavicle_l [{shapeType=Capsule, length=10.5, radius=8, shapeName=clavicle_l_capsule, center=<6.9631 0 0>, rotation=<rot 90 0 0>}]
 deform_wire [{shapeType=Capsule, length=5, radius=5, shapeName=deform_wire_capsule}]
 dyn_wire [{shapeType=Capsule, length=5, radius=5, shapeName=dyn_wire_capsule, center=<6.4905 -4.7866 0>, rotation=<rot 90 0 0>}]
 dyn_wire_1 [{shapeType=Capsule, length=5, radius=5, shapeName=dyn_wire_1_capsule, center=<0 -3.0915 -0.0115>, rotation=<rot 90 0 0>}]
 dyn_wire_2 [{
   shapeType=Capsule
   length=3.9927
   radius=2.14
   shapeName=dyn_wire_2_capsule
   center=<3.6784 -0.6451 -0.3658>
   rotation=<rot 0 -77.4586 -74.5684>
  }]
 head [{shapeType=Capsule, length=0.5, radius=8, shapeName=head_capsule, rotation=<rot 90 -90.0004 0>}]
 neck_01 [{shapeType=Capsule, length=6, radius=7, shapeName=neck_01_capsule, center=<0.8261 0 0>, rotation=<rot 90 0 0>}]
 spine_05 [{shapeType=Capsule, length=23, radius=10.5, shapeName=spine_05_capsule, center=<2.487 0 0>}]
}
constraints [
 {
  bone1Name=head
  bone2Name=neck_01
  swing1LimitDegrees=45
  swing1Motion=Free
  swing2LimitDegrees=45
  swing2Motion=Free
  twistLimitDegrees=45
  twistMotion=Free
 }
 {
  bone1Name=neck_01
  bone2Name=spine_05
  swing1LimitDegrees=45
  swing1Motion=Free
  swing2LimitDegrees=45
  swing2Motion=Free
  twistLimitDegrees=45
  twistMotion=Free
 }
 {
  bone1Name=deform_wire
  bone2Name=head
  swing1LimitDegrees=45
  swing1Motion=Free
  swing2LimitDegrees=45
  swing2Motion=Free
  twistLimitDegrees=45
  twistMotion=Free
 }
 {bone1Name=dyn_wire, bone2Name=deform_wire, swing1LimitDegrees=45, swing2LimitDegrees=15, twistLimitDegrees=15}
 {bone1Name=dyn_wire_1, bone2Name=dyn_wire, swing1LimitDegrees=45, swing2LimitDegrees=45, twistLimitDegrees=45}
 {bone1Name=dyn_wire_2, bone2Name=spine_05, swing1LimitDegrees=15, swing2LimitDegrees=45, twistLimitDegrees=1}
 {bone1Name=dyn_wire_2, bone2Name=dyn_wire_1, swing1LimitDegrees=5, swing2LimitDegrees=5, twistLimitDegrees=5}
 {
  bone1Name=clavicle_l
  bone2Name=spine_05
  swing1LimitDegrees=45
  swing1Motion=Free
  swing2LimitDegrees=45
  swing2Motion=Free
  twistLimitDegrees=45
  twistMotion=Free
 }
]
