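# Material function for the ByteGear character body materials that outputs a pseudo-random, time-driven 0–1 flicker float.
# Six sine waves run on Time at parameterized rates, scaled through named reroutes, with TimeMult and LerpTimeMult exposed as defaulted inputs.
# The sum of one sine pair and the difference of another are blended by the product of a third, then saturated into RandomFlickerFloat.
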MaterialFunction /BRCosmetics/Characters/Player/Male/Medium/Bodies/M_MED_ByteGear/Materials/MF__FloatFlickerRandom_WithTimeMult
graph {nodes [
  n1 = NamedRerouteUsage()
  n2 = Multiply(A=Time(), B=ScalarParameter() {defaultValue=4.75, group=RandomFloatFlicker, parameterName=RandomFlicker_Value01})
  n3 = NamedRerouteUsage()
  n4 = Multiply(A=Time(), B=ScalarParameter() {defaultValue=3.56, group=RandomFloatFlicker, parameterName=RandomFlicker_Value02})
  n5 = Add(A=Sine(None=Multiply(A=@n1, B=@n2)), B=Sine(None=Multiply(A=@n3, B=@n4)))
  n6 = NamedRerouteUsage()
  n7 = Multiply(A=Time(), B=ScalarParameter() {defaultValue=0.23, group=RandomFloatFlicker, parameterName=RandomFlicker_Value03})
  n8 = NamedRerouteUsage()
  n9 = Multiply(A=Time(), B=ScalarParameter() {defaultValue=0.75, group=RandomFloatFlicker, parameterName=RandomFlicker_Value04})
  n10 = Subtract(A=Sine(None=Multiply(A=@n6, B=@n7)), B=Sine(None=Multiply(A=@n8, B=@n9)))
  n11 = NamedRerouteUsage()
  n12 = Multiply(A=Time(), B=ScalarParameter() {defaultValue=0.33, group=RandomFloatFlicker, parameterName=RandomFlicker_Value05})
  n13 = NamedRerouteUsage()
  n14 = Multiply(A=Time(), B=ScalarParameter() {defaultValue=0.647, group=RandomFloatFlicker, parameterName=RandomFlicker_Value06})
  n15 = Multiply(A=Sine(None=Multiply(A=@n11, B=@n12)), B=Sine(None=Multiply(A=@n13, B=@n14)))
  FunctionOutput(None=Saturate(None=LinearInterpolate(A=@n5, B=@n10, Alpha=@n15))) {outputName=RandomFlickerFloat}
  n16 = FunctionInput() {bUsePreviewValueAsDefault=true, inputName=TimeMult, inputType=FunctionInput_Scalar, previewValue=<0.2 0 0 1>}
  NamedRerouteDeclaration(None=@n16) {name=TimeMult, nodeColor=<rgba 1 0.2109 0 1>}
  n17 = FunctionInput() {bUsePreviewValueAsDefault=true, inputName=LerpTimeMult, inputType=FunctionInput_Scalar, sortPriority=1, previewValue=<0.2 0 0 1>}
  NamedRerouteDeclaration(None=@n17) {name=LerpMult, nodeColor=<rgba 0 0.1094 1 1>}
 ]}
