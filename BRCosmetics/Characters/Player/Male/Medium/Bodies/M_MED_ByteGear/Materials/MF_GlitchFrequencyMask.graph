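# Material function for the ByteGear character body materials that outputs a scalar, time-driven glitch mask.
# A sine of Time runs at a frequency input jittered by MF__FloatFlickerRandom_WithTimeMult, then is multiplied by a second random flicker.
# Four inputs expose glitch frequency, duration and their multipliers through named reroutes.
# A SmoothStep duration window and a sine remap to 0–1 are built but do not feed the output.
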MaterialFunction /BRCosmetics/Characters/Player/Male/Medium/Bodies/M_MED_ByteGear/Materials/MF_GlitchFrequencyMask
graph {
 properties {bExposeToLibrary=true, description="Creates a glitchy "}
 nodes [
  n1 = NamedRerouteUsage()
  n2 = MaterialFunctionCall!(TimeMult=Constant() {r=0.3}, LerpTimeMult) {materialFunction=&/MF__FloatFlickerRandom_WithTimeMult}
  n3 = Multiply(A=Subtract(A=Multiply(A=@n2.RandomFlickerFloat) {constB=2}), B=NamedRerouteUsage()) {constB=0.4}
  n4 = Sine(None=Multiply(A=Time(), B=Add(A=@n1, B=@n3)) {constB=0.2})
  Add(A=Multiply(A=@n4) {constB=0.5}) {constB=0.5}
  n5 = Constant() {r=0.2}
  FunctionOutput(None=Multiply(
    A=@n4
    B=MaterialFunctionCall(TimeMult=@n5) {materialFunction=&MF__FloatFlickerRandom_WithTimeMult}.RandomFlickerFloat
   ))
  NamedRerouteDeclaration(None=FunctionInput(Preview=Constant() {r=0.6}) {bUsePreviewValueAsDefault=true, inputName=GlitchFrequency, inputType=FunctionInput_Scalar}) {name=GlitchFrequency, nodeColor=<rgba 0.0547 1 0 1>}
  NamedRerouteDeclaration(None=FunctionInput(Preview=Constant() {r=0.4}) {bUsePreviewValueAsDefault=true, inputName=GlitchFrequencyMultiplier, inputType=FunctionInput_Scalar, sortPriority=1}) {name=GlitchFrequencyMultiplier, nodeColor=<rgba 0.0547 1 0 1>}
  NamedRerouteDeclaration(None=FunctionInput(Preview=Constant() {r=0.4}) {bUsePreviewValueAsDefault=true, inputName=GlitchDuration, inputType=FunctionInput_Scalar, sortPriority=2}) {name=GlitchDuration, nodeColor=<rgba 0.0547 1 0 1>}
  NamedRerouteDeclaration(None=FunctionInput(Preview=Constant() {r=0.2}) {bUsePreviewValueAsDefault=true, inputName=GlitchDurationMultiplier, inputType=FunctionInput_Scalar, sortPriority=3}) {name=GlitchDurationMultiplier, nodeColor=<rgba 0.0547 1 0 1>}
  n6 = Constant() {r=1}
  n7 = NamedRerouteUsage()
  n8 = MaterialFunctionCall(TimeMult=Constant() {r=0.3}) {materialFunction=&MF__FloatFlickerRandom_WithTimeMult}
  n9 = Multiply(A=Subtract(A=Multiply(A=@n8.RandomFlickerFloat) {constB=2}), B=NamedRerouteUsage()) {constB=0.2}
  SmoothStep(Min=Subtract(A=@n6, B=Add(A=@n7, B=@n9)), Max=@n6, Value=@n4) {constMax=0.9, constMin=0.8}
 ]
}
